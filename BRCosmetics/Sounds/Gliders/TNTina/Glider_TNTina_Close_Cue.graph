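# Sound cue for the TNTina glider's close sound, using glider land sound class, open/close attenuation and concurrency.
# A mixer layers a fixed close wave with a boosted random layer of one of three equally weighted distant explosion sounds.
# The explosions are reused BlackMonday weapon bang assets.
# The whole cue plays slightly lowered in pitch and volume.
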SoundCue /BRCosmetics/Sounds/Gliders/TNTina/Glider_TNTina_Close_Cue
root=Mixer_0
nodes [
 {id=Mixer_0, children [WavePlayer_0,Random_1], properties {inputVolume [0.7,1.4]}}
 {id=Random_1, children [WavePlayer_4,WavePlayer_5,WavePlayer_6], properties {weights [1,1,1]}}
 {id=WavePlayer_0, properties {soundWaveAssetPtr=&/Glider_TNTina_Close_02}}
 {id=WavePlayer_4, properties {soundWaveAssetPtr=/Game/Sounds/Fort_Weapons/BlackMonday/Bangs/BlackMonday_Bang_Explode_Distant_01}}
 {id=WavePlayer_5, properties {soundWaveAssetPtr=/Game/Sounds/Fort_Weapons/BlackMonday/Bangs/BlackMonday_Bang_Explode_Distant_02}}
 {id=WavePlayer_6, properties {soundWaveAssetPtr=/Game/Sounds/Fort_Weapons/BlackMonday/Bangs/BlackMonday_Bang_Explode_Distant_03}}
]
properties {
 attenuationSettings=/Game/Sounds/Attenuation/Fort_Player_Glider_OpenClose_Athena_Att
 audioPropertiesSheet=/BRCosmetics/Sounds/Settings/PropertySheets/Glider/APS_Cosmetic_Glider_Close
 maxDistance=10000
 pitchMultiplier=0.9
 soundClassObject=&Fort_Gliders_Land
 volumeMultiplier=0.9
 concurrencySet [&Athena_GliderOpenClose_Conc]
}
