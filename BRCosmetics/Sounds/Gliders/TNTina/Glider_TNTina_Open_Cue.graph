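# Sound cue that plays when the TNTina cosmetic glider opens.
# A mixer layers an open sample, slightly delayed and boosted with randomized volume at fixed pitch, with an undelayed, quieter whistle sample.
# Cue-level overrides set logarithmic distance attenuation with low-pass filtering and a small concurrency cap that prevents new instances.
# It uses the Gliders Land sound class and a shared glider-open property sheet.
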SoundCue /BRCosmetics/Sounds/Gliders/TNTina/Glider_TNTina_Open_Cue
root=Mixer_0
nodes [
 {id=Delay_0, children [Modulator_0], properties {delayMax=0.14, delayMin=0.14}}
 {id=Mixer_0, children [Delay_0,WavePlayer_6], properties {inputVolume [0.9,0.7]}}
 {id=Modulator_0, children [WavePlayer_1], properties {pitchMax=1, pitchMin=1, volumeMax=1.25, volumeMin=1.15}}
 {id=WavePlayer_1, properties {soundWaveAssetPtr=&/Glider_TNTina_Open_04}}
 {id=WavePlayer_6, properties {soundWaveAssetPtr=&/Glider_TNTina_Whistle_04}}
]
properties {
 attenuationSettings=/Game/Sounds/Attenuation/Fort_Player_Glider_OpenClose_Athena_Att
 audioPropertiesSheet=/BRCosmetics/Sounds/Settings/PropertySheets/Glider/APS_Cosmetic_Glider_Open
 bOverrideAttenuation=true
 bOverrideConcurrency=true
 maxDistance=10400
 soundClassObject=&Fort_Gliders_Land
 volumeMultiplier=1
 attenuationOverrides {
  bAttenuateWithLPF=true
  distanceAlgorithm=Logarithmic
  falloffDistance=10000
  lPFFrequencyAtMax=1000
  lPFRadiusMax=4000
  lPFRadiusMin=500
  reverbDistanceMax=3000
  reverbDistanceMin=300
  stereoSpread=512
 }
 concurrencyOverrides {maxCount=3, resolutionRule=PreventNew}
 concurrencySet [&Athena_GliderOpenClose_Conc]
}
